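# Sound cue for a portal expansion sound in Fortnite player gameplay audio.
# A single wave player plays the Portal_Expand_01 wave directly, with no randomization or modulation nodes.
# It uses the forged-consumable in-world attenuation and the Athena ambient sound class, with a slight volume boost.
# An overridden concurrency limit caps simultaneous instances, stopping the oldest when exceeded.
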SoundCue /Game/Sounds/Fort_GamePlay_Sounds/Player/Portal/Portal_Expand_Cue
root=WavePlayer_0
nodes [{id=WavePlayer_0, properties {soundWaveAssetPtr=&/Portal_Expand_01}}]
properties {
 attenuationSettings=/Game/Sounds/Attenuation/Fort_ForgedConsumable_InWorld_Att
 bOverrideConcurrency=true
 maxDistance=1000
 soundClassObject=&Fort_Ambient_Athena
 volumeMultiplier=1.039
 concurrencyOverrides {maxCount=8, resolutionRule=StopOldest}
}
