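# Looping ambient SoundCue for an in-world portal, with pitch and volume driven by a continuous "Open" parameter.
# A quality node picks the portal loop for its first two levels and an optimized Geode Oddities loop for the last two.
# The loop passes through smoothed Doppler into the modulator.
# Overridden attenuation adds logarithmic falloff and listener focus, and concurrency stops the quietest instance at the limit.
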SoundCue /Game/Sounds/Fort_GamePlay_Sounds/Player/Portal/Portal_InWorld_Loop_Cue
root=ModulatorContinuous_0
nodes [
 {id=Doppler_0, children [QualityLevel_0], properties {bUseSmoothing=true, dopplerIntensity=5}}
 {
  id=ModulatorContinuous_0
  children [Doppler_0]
  properties {pitchModulationParams {maxOutput=1.2, minOutput=0.9, parameterName=Open}, volumeModulationParams {minOutput=0.5, parameterName=Open}}
 }
 {id=QualityLevel_0, children [WavePlayer_3,WavePlayer_3,WavePlayer_2,WavePlayer_2]}
 {id=WavePlayer_2, properties {bLooping=true, soundWaveAssetPtr=/Game/Athena/Sounds/Events/Geode/Oddities/Geode_Oddities_InWorldLoop_01_opt}}
 {id=WavePlayer_3, properties {bLooping=true, soundWaveAssetPtr=&/Portal_InWorldLoop_01}}
]
properties {
 bOverrideAttenuation=true
 bOverrideConcurrency=true
 maxDistance=2000
 soundClassObject=&Fort_Ambient_Athena
 volumeMultiplier=1.559
 attenuationOverrides {
  bEnableFocusInterpolation=true
  bEnableListenerFocus=true
  distanceAlgorithm=Logarithmic
  falloffDistance=2000
  nonFocusVolumeAttenuation=0.5
  stereoSpread=1024
  attenuationShapeExtents {x=0}
 }
 concurrencyOverrides {maxCount=8, resolutionRule=StopQuietest}
}
