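# Sound cue for the first-person portal teleport sound played when the player enters or exits a portal.
# A single wave player feeds Portal_Teleport_01 into a Modulator at its default settings, which adds slight random pitch and volume variation.
# The cue plays at reduced volume in the Athena ambient sound class, with no attenuation set and an effectively world-sized max distance.
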SoundCue /Game/Sounds/Fort_GamePlay_Sounds/Player/Portal/Portal_Teleport_EnterExit_1P_Cue
root=Modulator_0
nodes [
 {id=Modulator_0, children [WavePlayer_0]}
 {id=WavePlayer_0, properties {soundWaveAssetPtr=&/Portal_Teleport_01}}
]
properties {maxDistance=2097152, soundClassObject=&Fort_Ambient_Athena, volumeMultiplier=0.624}
