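# Sound cue for a player entering a portal teleport, under the Fortnite player gameplay sounds.
# A single wave player plays the Portal_Teleport_01 sound wave with no randomization or mixing.
# The cue raises volume and pitch slightly, routes to the Athena ambient sound class and uses a portal enter/exit third-person attenuation asset for spatial falloff.
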SoundCue /Game/Sounds/Fort_GamePlay_Sounds/Player/Portal/Portal_Teleport_Enter_Cue
root=WavePlayer_0
nodes [{id=WavePlayer_0, properties {soundWaveAssetPtr=&/Portal_Teleport_01}}]
properties {
 attenuationSettings=/Game/Sounds/Attenuation/Fort_Portal_EnterExit_3P_Athena_Att
 maxDistance=3584
 pitchMultiplier=1.05
 soundClassObject=&Fort_Ambient_Athena
 volumeMultiplier=1.247
}
